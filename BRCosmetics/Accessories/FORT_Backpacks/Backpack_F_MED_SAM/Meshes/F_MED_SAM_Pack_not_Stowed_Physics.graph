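# Physics asset for the F_MED_SAM backpack (not-stowed mesh), setting up two dangling ribbon chains for physical motion.
# Capsule bodies cover the pack, pelvis, lower spine and every ribbon bone, with the torso shapes likely serving as colliders for the ribbons.
# Upper and lower chains branch from a shared ribbon root constrained to the pack, with swing and twist limits generally loosening toward the tips.
# The spine constraints are left free.
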PhysicsAsset /BRCosmetics/Accessories/FORT_Backpacks/Backpack_F_MED_SAM/Meshes/F_MED_SAM_Pack_not_Stowed_Physics
bodies {
 dyb_ribbon_root [{shapeType=Capsule, length=2, radius=2, shapeName=dyb_ribbon_root_capsule}]
 dyn_ribbon_lower_01 [{shapeType=Capsule, length=1, radius=2.2, shapeName=dyn_ribbon_lower_01_capsule, center=<2.139 -0.8075 -0.0001>}]
 dyn_ribbon_lower_02 [{shapeType=Capsule, length=0.561, radius=1.8656, shapeName=dyn_ribbon_lower_02_capsule, center=<2.1688 0 -0.0002>}]
 dyn_ribbon_lower_03 [{shapeType=Capsule, length=0.9567, radius=1.8939, shapeName=dyn_ribbon_lower_03_capsule, center=<2.1161 0 0.0001>}]
 dyn_ribbon_lower_04 [{shapeType=Capsule, length=1, radius=2.2, shapeName=dyn_ribbon_lower_04_capsule, center=<2.139 -0.8075 -0.0001>}]
 dyn_ribbon_lower_05 [{shapeType=Capsule, length=1, radius=2.2, shapeName=dyn_ribbon_lower_05_capsule, center=<2.139 -0.8075 -0.0001>}]
 dyn_ribbon_lower_06 [{shapeType=Capsule, length=1, radius=2.2, shapeName=dyn_ribbon_lower_06_capsule, center=<2.139 -0.8075 -0.0001>}]
 dyn_ribbon_lower_07 [{shapeType=Capsule, length=1, radius=2.2, shapeName=dyn_ribbon_lower_07_capsule, center=<2.139 -0.8075 -0.0001>}]
 dyn_ribbon_lower_08 [{shapeType=Capsule, length=1, radius=2.2, shapeName=dyn_ribbon_lower_08_capsule, center=<2.139 -0.8075 -0.0001>}]
 dyn_ribbon_lower_09 [{shapeType=Capsule, length=1, radius=2.2, shapeName=dyn_ribbon_lower_09_capsule, center=<2.139 -0.8075 -0.0001>}]
 dyn_ribbon_top_02 [{shapeType=Capsule, length=1, radius=2.3, shapeName=dyn_ribbon_top_04_capsule, center=<2.46 -0.8065 0>}]
 dyn_ribbon_top_03 [{shapeType=Capsule, length=1, radius=2.3, shapeName=dyn_ribbon_top_04_capsule, center=<2.46 -0.8065 0>}]
 dyn_ribbon_top_04 [{shapeType=Capsule, length=1, radius=2.3, shapeName=dyn_ribbon_top_04_capsule, center=<2.46 -0.8065 0>}]
 dyn_ribbon_top_05 [{shapeType=Capsule, length=1, radius=2.3, shapeName=dyn_ribbon_top_05_capsule, center=<2.46 -0.8065 0>}]
 dyn_ribbon_top_06 [{shapeType=Capsule, length=1, radius=2.3, shapeName=dyn_ribbon_top_06_capsule, center=<2.46 -0.8065 0>}]
 dyn_ribbon_top_07 [{shapeType=Capsule, length=1, radius=2.3, shapeName=dyn_ribbon_top_07_capsule, center=<2.46 -0.8065 0>}]
 dyn_upper_top_01 [{shapeType=Capsule, length=1, radius=2.3, shapeName=dyn_upper_top_01_capsule, center=<2.46 0 0>}]
 pack [{shapeType=Capsule, length=11.3963, radius=10.4429, shapeName=pack_capsule, center=<0 -1.9625 -2.4059>}]
 pelvis [{shapeType=Capsule, length=7.9439, radius=16.25, shapeName=pelvis_capsule, center=<-4.0715 0.5369 0>}]
 spine_01 [{shapeType=Capsule, length=7.1125, radius=12.6516, shapeName=spine_01_capsule, center=<0 -0.812 0.0003>}]
 spine_02 [{shapeType=Capsule, length=10.4205, radius=11.6497, shapeName=spine_02_capsule, center=<2.1627 -2.1579 0.0003>}]
}
constraints [
 {bone1Name=dyb_ribbon_root, bone2Name=pack, swing1LimitDegrees=10, swing2LimitDegrees=10, twistLimitDegrees=5}
 {bone1Name=dyn_upper_top_01, bone2Name=dyb_ribbon_root, swing1LimitDegrees=25, swing2LimitDegrees=25, twistLimitDegrees=10}
 {bone1Name=dyn_ribbon_top_02, bone2Name=dyn_upper_top_01, swing1LimitDegrees=15, swing2LimitDegrees=15, twistLimitDegrees=5}
 {bone1Name=dyn_ribbon_top_03, bone2Name=dyn_ribbon_top_02, swing1LimitDegrees=20, swing2LimitDegrees=20, twistLimitDegrees=7}
 {bone1Name=dyn_ribbon_top_04, bone2Name=dyn_ribbon_top_03, swing1LimitDegrees=50, swing2LimitDegrees=50, twistLimitDegrees=10}
 {bone1Name=dyn_ribbon_top_05, bone2Name=dyn_ribbon_top_04, swing1LimitDegrees=50, swing2LimitDegrees=50, twistLimitDegrees=10}
 {bone1Name=dyn_ribbon_top_06, bone2Name=dyn_ribbon_top_05, swing1LimitDegrees=50, swing2LimitDegrees=50, twistLimitDegrees=10}
 {bone1Name=dyn_ribbon_top_07, bone2Name=dyn_ribbon_top_06, swing1LimitDegrees=50, swing2LimitDegrees=50, twistLimitDegrees=10}
 {bone1Name=dyn_ribbon_lower_01, bone2Name=dyb_ribbon_root, swing1LimitDegrees=10, swing2LimitDegrees=10, twistLimitDegrees=5}
 {bone1Name=dyn_ribbon_lower_02, bone2Name=dyn_ribbon_lower_01, swing1LimitDegrees=15, swing2LimitDegrees=15, twistLimitDegrees=5}
 {bone1Name=dyn_ribbon_lower_03, bone2Name=dyn_ribbon_lower_02, swing1LimitDegrees=20, swing2LimitDegrees=20, twistLimitDegrees=7}
 {bone1Name=dyn_ribbon_lower_04, bone2Name=dyn_ribbon_lower_03, swing1LimitDegrees=50, swing2LimitDegrees=50, twistLimitDegrees=10}
 {bone1Name=dyn_ribbon_lower_05, bone2Name=dyn_ribbon_lower_04, swing1LimitDegrees=50, swing2LimitDegrees=50, twistLimitDegrees=10}
 {bone1Name=dyn_ribbon_lower_06, bone2Name=dyn_ribbon_lower_05, swing1LimitDegrees=50, swing2LimitDegrees=50, twistLimitDegrees=10}
 {bone1Name=dyn_ribbon_lower_07, bone2Name=dyn_ribbon_lower_06, swing1LimitDegrees=50, swing2LimitDegrees=50, twistLimitDegrees=10}
 {bone1Name=dyn_ribbon_lower_08, bone2Name=dyn_ribbon_lower_07, swing1LimitDegrees=50, swing2LimitDegrees=50, twistLimitDegrees=10}
 {bone1Name=dyn_ribbon_lower_09, bone2Name=dyn_ribbon_lower_08, swing1LimitDegrees=50, swing2LimitDegrees=50, twistLimitDegrees=10}
 {
  bone1Name=spine_01
  bone2Name=pelvis
  swing1LimitDegrees=45
  swing1Motion=Free
  swing2LimitDegrees=45
  swing2Motion=Free
  twistLimitDegrees=45
  twistMotion=Free
 }
 {
  bone1Name=spine_02
  bone2Name=spine_01
  swing1LimitDegrees=45
  swing1Motion=Free
  swing2LimitDegrees=45
  swing2Motion=Free
  twistLimitDegrees=45
  twistMotion=Free
 }
]
