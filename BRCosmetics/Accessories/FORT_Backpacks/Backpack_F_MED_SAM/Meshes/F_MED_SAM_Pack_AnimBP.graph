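# Animation Blueprint for the F_MED_SAM backpack cosmetic, switching its look and physics depending on whether the SAM-tagged pickaxe is stowed.
# Each update copies the parent mesh's world-space pelvis and two spine transforms onto the pack bones and refreshes a gravity override.
# Two RigidBody nodes with separate stowed and not-stowed physics assets follow, weighted by opposite alphas.
# Weapon and skydiving events flip those alphas and material hide parameters, treating skydiving as stowed.
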AnimBlueprint /BRCosmetics/Accessories/FORT_Backpacks/Backpack_F_MED_SAM/Meshes/F_MED_SAM_Pack_AnimBP
target_skeleton=&/F_MED_SAM_Pack_Skeleton
graphs [
 AnimGraph {nodes [
   Copy Pose From Mesh()
   n1 = "Transform (Modify) Bone - Bone: pelvis"(
    ComponentPose=Local To Component(LocalPose=Input Pose())
    Alpha=1
    Translation=Get pelvis_trans() -> (pelvis_trans:Vector)
    Rotation=Get pelvis_rot() -> (pelvis_rot:Rotator)
    Scale="1,1,1"
   ) {
    rotationMode=BMM_Replace
    rotationSpace=BCS_WorldSpace
    translationMode=BMM_Replace
    translationSpace=BCS_WorldSpace
    alphaBoolBlend {blendOption=Linear}
    boneToModify {boneName=pelvis}
   }
   n2 = "Transform (Modify) Bone - Bone: spine_01"(
    ComponentPose=@n1
    Alpha=1
    Translation=Get spine_01_trans() -> (spine_01_trans:Vector)
    Rotation=Get spine_01_rot() -> (spine_01_rot:Rotator)
    Scale="1,1,1"
   ) {
    rotationMode=BMM_Replace
    rotationSpace=BCS_WorldSpace
    translationMode=BMM_Replace
    translationSpace=BCS_WorldSpace
    alphaBoolBlend {blendOption=Linear}
    boneToModify {boneName=spine_01}
   }
   n3 = "Transform (Modify) Bone - Bone: spine_02"(
    ComponentPose=@n2
    Alpha=1
    Translation=Get spine_02_trans() -> (spine_02_trans:Vector)
    Rotation=Get spine_02_rot() -> (spine_02_rot:Rotator)
    Scale="1,1,1"
   ) {
    rotationMode=BMM_Replace
    rotationSpace=BCS_WorldSpace
    translationMode=BMM_Replace
    translationSpace=BCS_WorldSpace
    alphaBoolBlend {blendOption=Linear}
    boneToModify {boneName=spine_02}
   }
   n4 = RigidBody!(
    ComponentPose=@n3
    Alpha=Get is_weapon_stowed() -> (is_weapon_stowed:double)
    OverrideWorldGravity:Vector=Get gravity_override() -> (gravity_override:Vector)
    ExternalForce
   ) -> (Pose) {
    bClampLinearTranslationLimitToRefPose=true
    bOverrideWorldGravity=true
    overridePhysicsAsset=&/F_MED_SAM_Pack_Physics
    simulationSpace=BaseBoneSpace
    alphaBoolBlend {blendOption=Linear}
    baseBoneRef {boneName=pack}
    componentAppliedLinearAccClamp=<3000 3000 3000>
    componentLinearAccScale=<0.4 0.3 0.3>
    componentLinearVelScale=<0.4 0.3 0.3>
    overrideWorldGravity {z=-980}
   }
   n5 = RigidBody(ComponentPose=@n4, Alpha=Get is_weapon_stowed_inv() -> (is_weapon_stowed_inv:double), OverrideWorldGravity="0,0,-980") {
    bClampLinearTranslationLimitToRefPose=true
    bOverrideWorldGravity=true
    overridePhysicsAsset=&/F_MED_SAM_Pack_not_Stowed_Physics
    simulationSpace=BaseBoneSpace
    alphaBoolBlend {blendOption=Linear}
    baseBoneRef {boneName=pack}
    componentAppliedLinearAccClamp=<3000 3000 3000>
    componentLinearAccScale=<0.4 0.3 0.3>
    componentLinearVelScale=<0.4 0.3 0.3>
    overrideWorldGravity {z=-980}
   }
   Output Pose(Result=Component To Local(ComponentPose=@n5)) {layerGroup=None}
  ]}
 EventGraph {nodes [
   Event BlueprintUpdateAnimation()
   > n1 = Sequence()
   > n2 = Gravity Override State Function DEPRECATED(CharacterPartAnimInstance=Self-Reference(), GameTimeInSeconds=GetGameTimeInSeconds())
   > Set gravity_override(gravity_override=@n2)
   n3 = GetPartSkeletalMeshComponent()
   Is Valid(exec=@n1.then_1, InputObject=@n3)
   n4 = GetAttachParent!(self:Scene Component Object Reference=@n3) -> (ReturnValue:Scene Component Object Reference)
   > Is Valid(InputObject=@n4)
   n5 = BreakTransform(InTransform=GetSocketTransform(self=@n4, InSocketName=Pelvis, TransformSpace=RTS_World))
   > Set pelvis_trans(pelvis_trans:Vector=@n5) -> (Output_Get:Vector)
   > Set pelvis_rot(pelvis_rot:Rotator=@n5.Rotation) -> (Output_Get:Rotator)
   n6 = BreakTransform(InTransform=GetSocketTransform(self=@n4, InSocketName=Spine_01, TransformSpace=RTS_World))
   > Set spine_01_trans(spine_01_trans:Vector=@n6) -> (Output_Get:Vector)
   > Set spine_01_rot(spine_01_rot:Rotator=@n6.Rotation) -> (Output_Get:Rotator)
   n7 = BreakTransform(InTransform=GetSocketTransform(self=@n4, InSocketName=Spine_02, TransformSpace=RTS_World))
   > Set spine_02_trans(spine_02_trans:Vector=@n7) -> (Output_Get:Vector)
   > Set spine_02_rot(spine_02_rot:Rotator=@n7.Rotation) -> (Output_Get:Rotator)
   Event OnWeaponChanged()
   > n8 = Get Dynamic Material Instance(SkeletalMeshComponent=GetPartSkeletalMeshComponent())
   n9 = PickaxeItemDefinitionHasAnyMetaTag(Tags=Make Array("[0]"=(TagName="Cosmetics.MetaTags.SAM")), bRequiresEquipped=true)
   > n10 = Branch(Condition=@n9)
   > n11 = Branch(Condition=Get bIsSkyDiving())
   > n12 = Sequence()
   SetScalarParameterValue(execute=[@n18.then_0, @n12.then_0, @n14.then_0, @n15.then_0], self=@n8, ParameterName=Hide Element 01, Value=0)
   > Set is_weapon_stowed(is_weapon_stowed:double=1) -> (Output_Get:double)
   SetScalarParameterValue(execute=[@n18.then_1, @n12.then_1, @n14.then_1, @n15.then_1], self=@n8, ParameterName=Hide Element 02, Value=1)
   > Set is_weapon_stowed_inv(is_weapon_stowed_inv:double=0) -> (Output_Get:double)
   n13 = Sequence(execute=@n11.else)
   SetScalarParameterValue(execute=[@n13.then_0, @n17.then_0], self=@n8, ParameterName=Hide Element 01, Value=1)
   > Set is_weapon_stowed(is_weapon_stowed=0)
   SetScalarParameterValue(execute=[@n13.then_1, @n17.then_1], self=@n8, ParameterName=Hide Element 02, Value=0)
   > Set is_weapon_stowed_inv(is_weapon_stowed_inv=1)
   n14 = Sequence(execute=@n10.else)
   Event OnBeginSkydiving()
   > n15 = Sequence()
   Event OnEndSkydiving()
   > n16 = Branch(Condition=@n9)
   > n17 = Sequence()
   n18 = Sequence(execute=@n16.else)
  ]}
]
variables [
 gravity_override {type=Vector}
 is_weapon_stowed {type=number}
 is_weapon_stowed_inv {type=number}
 pelvis_rot {type=Rotator}
 pelvis_trans {type=Vector}
 spine_01_rot {type=Rotator}
 spine_01_trans {type=Vector}
 spine_02_rot {type=Rotator}
 spine_02_trans {type=Vector}
]
